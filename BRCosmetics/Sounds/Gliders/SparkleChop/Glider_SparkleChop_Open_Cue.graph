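# Sound cue that plays when the SparkleChop cosmetic glider opens.
# A mixer layers the SparkleChop open sound, slightly lowered in volume, with the reused DefaultMarkII glider open sound at full volume behind a short fixed delay.
# It uses glider open/close attenuation and concurrency, the glider-open cosmetic property sheet, and the Fort_Gliders_Land sound class.
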SoundCue /BRCosmetics/Sounds/Gliders/SparkleChop/Glider_SparkleChop_Open_Cue
root=Mixer_0
nodes [
 {id=Delay_0, children [WavePlayer_0], properties {delayMax=0.13, delayMin=0.13}}
 {id=Mixer_0, children [WavePlayer_1,Delay_0], properties {inputVolume [0.8,1]}}
 {id=WavePlayer_0, properties {soundWaveAssetPtr=/BRCosmetics/Sounds/Gliders/DefaultMarkII/Glider_Mark2_Open}}
 {id=WavePlayer_1, properties {soundWaveAssetPtr=&/Glider_SparkleChop_Open}}
]
properties {
 attenuationSettings=/Game/Sounds/Attenuation/Fort_Player_Glider_OpenClose_Athena_Att
 audioPropertiesSheet=/BRCosmetics/Sounds/Settings/PropertySheets/Glider/APS_Cosmetic_Glider_Open
 maxDistance=10000
 soundClassObject=&Fort_Gliders_Land
 volumeMultiplier=0.9
 concurrencySet [&Athena_GliderOpenClose_Conc]
}
